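# Niagara swing trail for the HammerVice pickaxe, drawing short-lived, blue-tinted ribbons from weapon sockets.
# Mirrored left and right emitters each spawn four anim trails, gated by User.SwingSectionIndex so swings alternate sides and section 3 fires both.
# Socket readers drive each trail, with per-group offsets switching from local Z to Y through System.isLastSwing.
# Ribbons narrow and fade over their lifetime, with colors picked by time of day.
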NiagaraSystem /BRCosmetics/Weapons/FORT_Melee/Pickaxe_HammerVice/FX/NS_Pickaxe_HammerVice_Trail
dependencies {
 usedDataInterfaces [&NiagaraDataInterfaceCurve, &NiagaraDataInterfaceColorCurve, &NiagaraDataInterfaceSocketReader]
 usedDynamicInputs [&MakeVectorFromLinearColorRGB, &MergeHelper_FloatSwitch]
 usedModules [
  &self:SystemSpawnScript.SystemScriptSource.SystemScriptGraph.NiagaraNodeAssignment_1.SetVariables_#1
  &self:SystemSpawnScript.SystemScriptSource.SystemScriptGraph.NiagaraNodeAssignment_2.SetVariables_#2
  &self:SystemSpawnScript.SystemScriptSource.SystemScriptGraph.NiagaraNodeAssignment_0.SetVariables_#3
  &SystemState
 ]
}
emitters {
 NE_Swing_Left {
  emitterSpawnScript [&FN_RibbonIDOffsetUtility()]
  emitterUpdateScript [
   &EmitterState(Life Cycle Mode=::NewEnumerator1, Loop Behavior=::NewEnumerator1, Loop Duration=0.175, UseLoopDelay=false)
   &FN_SpawnAnimTrail(Number of Anim Trails:NiagaraInt32=&MergeHelper_IntSwitch(0=0, 1=4, 2=0, 3=4, 4=?, 5=?, Index=$User.SwingSectionIndex, Num=::NewEnumerator2))
  ]
  particleSpawnScript [
   &FN_InitializeAnimTrailBoneIndecies()
   &FN_InitializeAnimTrailBoneIndecies(AnimTrail Selection=1, First Bone or Socket Index=2, Second Bone or Socket Index=3) {module=FN_InitializeAnimTrailBoneIndecies003}
   &FN_InitializeAnimTrailBoneIndecies(AnimTrail Selection=2, First Bone or Socket Index=4, Second Bone or Socket Index=5) {module=FN_InitializeAnimTrailBoneIndecies002}
   &FN_InitializeAnimTrailBoneIndecies(AnimTrail Selection=3, First Bone or Socket Index=6, Second Bone or Socket Index=7) {module=FN_InitializeAnimTrailBoneIndecies001}
   &FN_InitializeAnimTrail(
    AnimTrailWidth=0.3
    Color=&FNTimeOfDayColor(Night Color=<rgba 1.95 2.125 3 1>, Morning Color=<rgba 2.1 2.325 3 1>, Day Color=<rgba 2.25 2.5 3 1>, Evening Color=<rgba 2.1 2.325 3 1>)
    Lifetime=0.165
    Position Offset=&Multiply_VectorByFloat(
     Vector=&TransformVector(
      Vector=&Lerp_Vector(
       A=&MergeHelper_VectorSwitch(0=<0 0 -1>, 1=<0 0 -0.5>, 2=<0 0 0.5>, 3=<0 0 1>, Index=$Particles.SpawnGroup, Num=::NewEnumerator2)
       B=&MergeHelper_VectorSwitch(0=<0 -1 0>, 1=<0 -0.5 0>, 2=<0 0.5 0>, 3=<0 1 0>, Index=$Particles.SpawnGroup, Num=::NewEnumerator2)
       Alpha=$System.isLastSwing
      )
      Destination Space=::Local
     )
     Float=2.5
    )
    Socket Reader:NiagaraDataInterfaceSocketReader=$System.Socket Reader_Left
   )
  ]
  particleUpdateScript [
   &SolveForcesAndVelocity(Clamp Velocity=false, Limit Acceleration=false)
   &FN_UpdateAnimTrail(Socket Reader:NiagaraDataInterfaceSocketReader=$System.Socket Reader_Left) {module=FN_UpdateAnimTrail001}
   &ScaleRibbonWidth(Ribbon Width Scale=&FloatFromCurve(FloatCurve=DataInterface() {propertyValues {Curve {keys [{arriveTangent=-0.0042, interpMode=RCIM_Cubic, leaveTangent=-0.0042, tangentMode=RCTM_User, time=0, value=1}, <at 0.5 0>]}}}))
   &ScaleColor(
    Color Scale:LinearColor=&FNTimeOfDayColor(
     Night Color=&ColorFromCurve(ColorCurve=DataInterface() {propertyValues {
        AlphaCurve {keys [<at 0 1>,<at 0.175 0>]}
        BlueCurve {keys [<at 0 5>,<at 0.25 1>,<at 0.75 1>,<at 1 0>]}
        GreenCurve {keys [<at 0 3.6875>,<at 0.25 0.625>,<at 0.75 0.5417>,<at 1 0>]}
        RedCurve {keys [<at 0 3.25>,<at 0.25 0.55>,<at 0.75 0.5>,<at 1 0>]}
       }})
     Morning Color=&ColorFromCurve(ColorCurve=DataInterface() {propertyValues {
        AlphaCurve {keys [<at 0 1>,<at 0.175 0>]}
        BlueCurve {keys [<at 0 5>,<at 0.25 1>,<at 0.75 1>,<at 1 0>]}
        GreenCurve {keys [<at 0 3.6875>,<at 0.25 0.5875>,<at 0.75 0.5417>,<at 1 0>]}
        RedCurve {keys [<at 0 3.25>,<at 0.25 0.55>,<at 0.75 0.5>,<at 1 0>]}
       }})
     Day Color=&ColorFromCurve(ColorCurve=DataInterface() {propertyValues {
        AlphaCurve {keys [<at 0 1>,<at 0.175 0>]}
        BlueCurve {keys [<at 0 5>,<at 0.25 1>,<at 0.75 1>,<at 1 0>]}
        GreenCurve {keys [<at 0 3.8333>,<at 0.25 0.6625>,<at 0.75 0.5417>,<at 1 0>]}
        RedCurve {keys [<at 0 3.25>,<at 0.25 0.55>,<at 0.75 0.5>,<at 1 0>]}
       }})
     Evening Color=&ColorFromCurve(ColorCurve=DataInterface() {propertyValues {
        AlphaCurve {keys [<at 0 1>,<at 0.175 0>]}
        BlueCurve {keys [<at 0 5>,<at 0.25 1>,<at 0.75 1>,<at 1 0>]}
        GreenCurve {keys [<at 0 3.6875>,<at 0.25 0.5875>,<at 0.75 0.5417>,<at 1 0>]}
        RedCurve {keys [<at 0 3.25>,<at 0.25 0.55>,<at 0.75 0.5>,<at 1 0>]}
       }})
    )
    Scale Mode=3
   )
  ]
  rendererClasses [&NiagaraRibbonRendererProperties]
  renderers [{rendererClass=&NiagaraRibbonRendererProperties}]
 }
 NE_Swing_Right {
  emitterSpawnScript [&FN_RibbonIDOffsetUtility()]
  emitterUpdateScript [
   &EmitterState(Life Cycle Mode=::NewEnumerator1, Loop Behavior=::NewEnumerator1, Loop Duration=0.175, UseLoopDelay=false)
   &FN_SpawnAnimTrail(Number of Anim Trails:NiagaraInt32=&MergeHelper_IntSwitch(0=4, 1=0, 2=4, 3=4, 4=?, 5=?, Index=$User.SwingSectionIndex, Num=::NewEnumerator2))
  ]
  particleSpawnScript [
   &FN_InitializeAnimTrailBoneIndecies()
   &FN_InitializeAnimTrailBoneIndecies(AnimTrail Selection=1, First Bone or Socket Index=2, Second Bone or Socket Index=3) {module=FN_InitializeAnimTrailBoneIndecies003}
   &FN_InitializeAnimTrailBoneIndecies(AnimTrail Selection=2, First Bone or Socket Index=4, Second Bone or Socket Index=5) {module=FN_InitializeAnimTrailBoneIndecies002}
   &FN_InitializeAnimTrailBoneIndecies(AnimTrail Selection=3, First Bone or Socket Index=6, Second Bone or Socket Index=7) {module=FN_InitializeAnimTrailBoneIndecies001}
   &FN_InitializeAnimTrail(
    AnimTrailWidth=0.3
    Color=&FNTimeOfDayColor(Night Color=<rgba 1.95 2.125 3 1>, Morning Color=<rgba 2.1 2.325 3 1>, Day Color=<rgba 2.25 2.5 3 1>, Evening Color=<rgba 2.1 2.325 3 1>)
    Lifetime=0.165
    Position Offset=&Multiply_VectorByFloat(
     Vector=&TransformVector(
      Vector=&Lerp_Vector(
       A=&MergeHelper_VectorSwitch(0=<0 0 -1>, 1=<0 0 -0.5>, 2=<0 0 0.5>, 3=<0 0 1>, Index=$Particles.SpawnGroup, Num=::NewEnumerator2)
       B=&MergeHelper_VectorSwitch(0=<0 -1 0>, 1=<0 -0.5 0>, 2=<0 0.5 0>, 3=<0 1 0>, Index=$Particles.SpawnGroup, Num=::NewEnumerator2)
       Alpha=$System.isLastSwing
      )
      Destination Space=::Local
     )
     Float=2.5
    )
   )
  ]
  particleUpdateScript [
   &SolveForcesAndVelocity(Clamp Velocity=false, Limit Acceleration=false)
   &FN_UpdateAnimTrail() {module=FN_UpdateAnimTrail001}
   &ScaleRibbonWidth(Ribbon Width Scale=&FloatFromCurve(FloatCurve=DataInterface() {propertyValues {Curve {keys [{arriveTangent=-0.0042, interpMode=RCIM_Cubic, leaveTangent=-0.0042, tangentMode=RCTM_User, time=0, value=1}, <at 0.5 0>]}}}))
   &ScaleColor(
    Color Scale:LinearColor=&FNTimeOfDayColor(
     Night Color=&ColorFromCurve(ColorCurve=DataInterface() {propertyValues {
        AlphaCurve {keys [<at 0 1>,<at 0.175 0>]}
        BlueCurve {keys [<at 0 5>,<at 0.25 1>,<at 0.75 1>,<at 1 0>]}
        GreenCurve {keys [<at 0 3.6875>,<at 0.25 0.625>,<at 0.75 0.5417>,<at 1 0>]}
        RedCurve {keys [<at 0 3.25>,<at 0.25 0.55>,<at 0.75 0.5>,<at 1 0>]}
       }})
     Morning Color=&ColorFromCurve(ColorCurve=DataInterface() {propertyValues {
        AlphaCurve {keys [<at 0 1>,<at 0.175 0>]}
        BlueCurve {keys [<at 0 5>,<at 0.25 1>,<at 0.75 1>,<at 1 0>]}
        GreenCurve {keys [<at 0 3.6875>,<at 0.25 0.5875>,<at 0.75 0.5417>,<at 1 0>]}
        RedCurve {keys [<at 0 3.25>,<at 0.25 0.55>,<at 0.75 0.5>,<at 1 0>]}
       }})
     Day Color=&ColorFromCurve(ColorCurve=DataInterface() {propertyValues {
        AlphaCurve {keys [<at 0 1>,<at 0.175 0>]}
        BlueCurve {keys [<at 0 5>,<at 0.25 1>,<at 0.75 1>,<at 1 0>]}
        GreenCurve {keys [<at 0 3.8333>,<at 0.25 0.6625>,<at 0.75 0.5417>,<at 1 0>]}
        RedCurve {keys [<at 0 3.25>,<at 0.25 0.55>,<at 0.75 0.5>,<at 1 0>]}
       }})
     Evening Color=&ColorFromCurve(ColorCurve=DataInterface() {propertyValues {
        AlphaCurve {keys [<at 0 1>,<at 0.175 0>]}
        BlueCurve {keys [<at 0 5>,<at 0.25 1>,<at 0.75 1>,<at 1 0>]}
        GreenCurve {keys [<at 0 3.6875>,<at 0.25 0.5875>,<at 0.75 0.5417>,<at 1 0>]}
        RedCurve {keys [<at 0 3.25>,<at 0.25 0.55>,<at 0.75 0.5>,<at 1 0>]}
       }})
    )
    Scale Mode=3
   )
  ]
  rendererClasses [&NiagaraRibbonRendererProperties]
  renderers [{rendererClass=&NiagaraRibbonRendererProperties}]
 }
}
summary {
 systemName=NS_Pickaxe_HammerVice_Trail
 emitters [
  {emitterName=NE_Swing_Right, rendererClasses [&NiagaraRibbonRendererProperties]}
  {emitterName=NE_Swing_Left, rendererClasses [&NiagaraRibbonRendererProperties]}
 ]
 userVariables [
  User.RibbonIDOffset {type=NiagaraInt32}
  User.SwingSectionIndex {type=NiagaraInt32}
 ]
}
