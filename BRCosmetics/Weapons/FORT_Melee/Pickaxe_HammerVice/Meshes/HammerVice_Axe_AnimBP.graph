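# Animation Blueprint for the HammerVice pickaxe, a cosmetic melee weapon, driving the HammerVice_Axe skeleton.
# Copy Pose From Mesh pulls the pose from the attached parent mesh and feeds it into a Control Rig node running HammerVice_Axe_CtrlRig, whose result is the output pose.
# The event graph is empty and there are no variables, so the AnimGraph does all the work.
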AnimBlueprint /BRCosmetics/Weapons/FORT_Melee/Pickaxe_HammerVice/Meshes/HammerVice_Axe_AnimBP
target_skeleton=&/HammerVice_Axe_Skeleton
graphs [
 AnimGraph {nodes [Output Pose(Result=Control Rig(Source=Copy Pose From Mesh() {bUseAttachedParent=true}) {controlRigAssetReference {controlRigAsset=&/HammerVice_Axe_CtrlRig}}) {layerGroup=None}]}
 EventGraph {links [], nodes []}
]
variables []
